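# Unlit, two-sided translucent material for a spinning klaxon light beam on Niagara mesh particles in Storm Shield missions.
# RotateAboutAxis spins the mesh around its pivot over time, offset by pivot height, with FixRotateAboutAxisNormals correcting normals for fresnel.
# Fresnel scaled by dynamic Fill joins a powered V-gradient to form a mask blending two emissive colors.
# Particle color tints the glow, while particle alpha, depth fade and Opacity shape transparency.
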Material /Game/Missions/Primary/StormShield/Materials/M_KlaxonLight_Beam
graph {
 properties {
  bUsedWithNiagaraMeshParticles=true
  blendMode=BLEND_Translucent
  pixelDepthOffsetMode=PDOM_Legacy
  shadingModel=MSM_Unlit
  twoSided=true
  usedShadingModels=Unlit
 }
 nodes [
  n1 = ComponentMask(None=TextureCoordinate()) {g=true}
  n2 = Power(Base=@n1, Exp=ScalarParameter() {defaultValue=16, parameterName=Exponent}) {constExponent=16}
  Multiply(A=@n2, B=ScalarParameter() {defaultValue=1, parameterName=Boost})
  n3 = DynamicParameter() {paramNames [Fill,Param2,Param3,Param4]}
  n4 = VectorParameter() {parameterName=RotationAxis, defaultValue=<rgba 0 0 1 0>}
  n5 = Multiply(A=Time()) {constB=0.7}
  n6 = ComponentMask(None=MaterialFunctionCall() {materialFunction=&ComponentPivotLocation}.Result) {b=true}
  n7 = MaterialFunctionCall(
   Brightness=ScalarParameter() {defaultValue=1, parameterName=Fresnel Brightness}
   Exponent=ScalarParameter() {defaultValue=4, parameterName=Fresnel Exponent}
   Vertex / Pixel Normal=MaterialFunctionCall(Rotation Axis=Transform(None=@n4.RGB) {transformSourceType=TRANSFORMSOURCE_Local}, Rotation Angle=Add(A=@n5, B=@n6)) {materialFunction=&FixRotateAboutAxisNormals}.New World Space Vertex Normal
  ) {materialFunction=&MF_Fresnel_FX}
  n8 = Multiply(A=Add(A=Multiply(A=@n3.Fill, B=@n7.Fresnel), B=@n2), B=Power(Base=@n1))
  Multiply(
   A=ParticleColor().RGB
   B=LinearInterpolate(
    A=VectorParameter() {parameterName=Emissive A1, defaultValue=<rgba 0.11 0.3028 1 0>}.RGB
    B=VectorParameter() {parameterName=Emissive B1, defaultValue=<rgba 0.826 1.2988 2 0>}.RGB
    Alpha=Clamp(None=@n8)
   )
  )
  n9 = ParticleColor()
  n10 = Constant() {r=25}
  Multiply(A=Multiply(A=Multiply(A=@n8, B=@n9.A), B=DepthFade(FadeDistance=@n10)), B=ScalarParameter() {defaultValue=0.3, parameterName=Opacity})
  n11 = Time()
  n12 = ScalarParameter() {defaultValue=1, parameterName=AddPivotLocationToRotateTime}
  n13 = MaterialFunctionCall() {materialFunction=&ComponentPivotLocation}
  n14 = ComponentMask(None=@n13.Result) {b=true}
  RotateAboutAxis(
   NormalizedRotationAxis=Transform(None=VectorParameter() {parameterName=RotationAxis, defaultValue=<rgba 0 0 1 0>}.RGB) {transformSourceType=TRANSFORMSOURCE_Local}
   RotationAngle=Add(A=Multiply(A=@n11) {constB=0.7}, B=Multiply(A=@n12, B=@n14))
   PivotPoint=@n13.Result
   Position=WorldPosition().XYZ
  )
 ]
}
