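# Unlit emissive material for the core of a Storm Shield klaxon light, usable on Niagara mesh particles, glowing blue by default with a rotating sweep.
# A time-scrolled sine along UV V, phase-shifted by component pivot height, forms the sweep.
# Sweep, U-coordinate masks and Fresnel blend two emissive colours.
# An inverted-Fresnel tint and a very bright, sharpened Fresnel-masked highlight following the sweep are added on top.
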Material /Game/Missions/Primary/StormShield/Materials/M_KlaxonLight_Core
graph {
 properties {bUsedWithNiagaraMeshParticles=true, pixelDepthOffsetMode=PDOM_Legacy, shadingModel=MSM_Unlit, usedShadingModels=Unlit}
 nodes [
  n1 = MaterialFunctionCall(Exponent=Constant() {r=2}) {materialFunction=&MF_Fresnel_FX}
  n2 = VectorParameter() {parameterName=FresCol, defaultValue=<rgba 0.3475 1.0721 2 0>}
  n3 = VectorParameter() {parameterName=Emissive A, defaultValue=<rgba 0.05 0.3817 10 0>}
  n4 = VectorParameter() {parameterName=Emissive B, defaultValue=<rgba 2.1189 2.7912 5 0>}
  n5 = MaterialFunctionCall(
   Brightness=ScalarParameter() {defaultValue=1, parameterName=Fresnel Brightness}
   Exponent=ScalarParameter() {defaultValue=5, parameterName=Fresnel Exponent}
  ) {materialFunction=&MF_Fresnel_FX}
  n6 = Multiply(A=ComponentMask(None=TextureCoordinate()) {r=true}) {constB=12}
  n7 = OneMinus(None=Clamp(None=Power(Base=@n6) {constExponent=4}))
  n8 = Multiply(A=ComponentMask(None=TextureCoordinate()) {g=true}) {constB=2}
  n9 = Multiply(A=Time(), B=ScalarParameter() {defaultValue=-1.4, parameterName=RotateSpeed}) {constB=-2}
  n10 = ScalarParameter() {defaultValue=-0.2, parameterName=RotateOffset}
  n11 = Multiply(A=ComponentMask(None=MaterialFunctionCall() {materialFunction=&ComponentPivotLocation}.Result) {b=true}) {constB=0.5}
  n12 = Sine(None=Add(A=Add(A=@n8, B=@n9), B=Add(A=@n10, B=@n11)))
  n13 = Add(A=@n12)
  n14 = Power(Base=ComponentMask(None=TextureCoordinate()) {r=true})
  n15 = Multiply(A=Clamp(None=Subtract(A=@n5.Fresnel, B=@n7)), B=Multiply(A=Clamp(None=@n13), B=@n14))
  n16 = Multiply(A=@n12) {constB=0.5}
  n17 = Multiply(
   A=Power(Base=Add(A=@n16) {constB=0.5}) {constExponent=4}
   B=MaterialFunctionCall(Exponent=Constant() {r=8}) {materialFunction=&MF_Fresnel_FX}.Fresnel
  )
  Add(
   A=Add(A=Multiply(A=@n1.Fresnel Inverted, B=@n2.RGB), B=LinearInterpolate(A=@n3.RGB, B=@n4.RGB, Alpha=@n15))
   B=Multiply(A=Multiply(A=@n14, B=@n17), B=VectorParameter() {parameterName=SuperBright, defaultValue=<rgba 5.12 26.832 64 0>}.RGB)
  )
 ]
}
