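# M_Missle_Light is a Default Lit material in the StormShield mission folder, named for a missile light, that produces a glowing two-color gradient.
# LinearGradient's VGradient output blends orange into red along the V texture coordinate.
# A constant multiplier brightens the blend before it feeds the Emissive input of the MF_FX_EmissiveAdjustments function.
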Material /Game/Missions/Primary/StormShield/Materials/M_Missle_Light
graph {
 properties {pixelDepthOffsetMode=PDOM_Legacy, usedShadingModels=Default Lit}
 nodes [
  Constant()
  n1 = Constant3Vector() {constant=<rgba 1 0.3404 0.025 0>}
  n2 = Constant3Vector() {constant=<rgba 1 0 0.0116 0>}
  n3 = MaterialFunctionCall() {materialFunction=&LinearGradient}
  MaterialFunctionCall(Emissive=Multiply(A=LinearInterpolate(A=@n1, B=@n2, Alpha=@n3.VGradient), B=Constant() {r=3})) {materialFunction=&MF_FX_EmissiveAdjustments}
 ]
}
